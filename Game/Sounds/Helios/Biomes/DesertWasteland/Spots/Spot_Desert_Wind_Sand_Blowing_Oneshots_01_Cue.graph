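# Sound cue for one-shot wind-blown sand spot ambience in the Desert Wasteland biome.
# A Random node picks one of five equally weighted waves, and a Modulator slightly varies its pitch and volume.
# A Quality Level root feeds that chain to its first two slots and leaves the other two empty.
# It plays at reduced volume through the outdoor ambient one-shot class with wind attenuation and a two-instance concurrency limit.
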SoundCue /Game/Sounds/Helios/Biomes/DesertWasteland/Spots/Spot_Desert_Wind_Sand_Blowing_Oneshots_01_Cue
root=QualityLevel_0
nodes [
 {id=Modulator_1, children [Random_0], properties {pitchMax=1.1, pitchMin=1, volumeMax=1.1, volumeMin=0.9}}
 {id=QualityLevel_0, children [Modulator_1,Modulator_1,null,null]}
 {id=Random_0, children [WavePlayer_2,WavePlayer_3,WavePlayer_4,WavePlayer_5,WavePlayer_8], properties {weights [1,1,1,1,1]}}
 {id=WavePlayer_2, properties {soundWaveAssetPtr=&/Spot_Desert_Wind_Sand_Emitter_Oneshot_01}}
 {id=WavePlayer_3, properties {soundWaveAssetPtr=&/Spot_Desert_Wind_Sand_Emitter_Oneshot_02}}
 {id=WavePlayer_4, properties {soundWaveAssetPtr=&/Spot_Desert_Wind_Sand_Emitter_Oneshot_03}}
 {id=WavePlayer_5, properties {soundWaveAssetPtr=&/Spot_Desert_Wind_Sand_Emitter_Oneshot_04}}
 {id=WavePlayer_8, properties {soundWaveAssetPtr=&/Spot_Desert_Wind_Sand_Emitter_Oneshot_05}}
]
properties {
 attenuationSettings=/Game/Sounds/Attenuation/Fort_TileAmbient_OneShot_Wind_Att
 bEnableBusSends=false
 bOverrideConcurrency=true
 maxDistance=8000
 soundClassObject=&Fort_Ambient_OneShots_Outdoors
 volumeMultiplier=0.4
 concurrencyOverrides {maxCount=2}
 preEffectBusSends [{}]
}
