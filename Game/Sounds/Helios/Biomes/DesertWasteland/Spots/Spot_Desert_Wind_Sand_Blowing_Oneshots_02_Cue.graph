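# Sound cue for desert wasteland ambient spots, playing one-shot wind and blowing sand sounds.
# A random node picks evenly among four emitter one-shot waves, then a modulator slightly varies pitch downward and volume around normal.
# All four quality level branches share that same chain.
# It plays quietly through a tile ambient wind attenuation and outdoor ambient sound class, with a concurrency override capping it at two voices.
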SoundCue /Game/Sounds/Helios/Biomes/DesertWasteland/Spots/Spot_Desert_Wind_Sand_Blowing_Oneshots_02_Cue
root=QualityLevel_0
nodes [
 {id=Modulator_1, children [Random_0], properties {pitchMax=1, pitchMin=0.9, volumeMax=1.1, volumeMin=0.9}}
 {id=QualityLevel_0, children [Modulator_1,Modulator_1,Modulator_1,Modulator_1]}
 {id=Random_0, children [WavePlayer_9, WavePlayer_10, WavePlayer_11, WavePlayer_12], properties {weights [1,1,1,1]}}
 {id=WavePlayer_10, properties {soundWaveAssetPtr=&/Spot_Desert_Wind_Sand_Emitter_Oneshot_07}}
 {id=WavePlayer_11, properties {soundWaveAssetPtr=&/Spot_Desert_Wind_Sand_Emitter_Oneshot_08}}
 {id=WavePlayer_12, properties {soundWaveAssetPtr=&/Spot_Desert_Wind_Sand_Emitter_Oneshot_09}}
 {id=WavePlayer_9, properties {soundWaveAssetPtr=&/Spot_Desert_Wind_Sand_Emitter_Oneshot_06}}
]
properties {
 attenuationSettings=/Game/Sounds/Attenuation/Fort_TileAmbient_OneShot_Wind_Att
 bEnableBusSends=false
 bOverrideConcurrency=true
 maxDistance=8000
 soundClassObject=&Fort_Ambient_OneShots_Outdoors
 volumeMultiplier=0.4
 concurrencyOverrides {maxCount=2, voiceStealReleaseTime=2}
 preEffectBusSends [{}]
}
