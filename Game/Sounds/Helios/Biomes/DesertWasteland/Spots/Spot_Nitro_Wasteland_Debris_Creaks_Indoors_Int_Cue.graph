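# Indoor ambient one-shot cue for debris falls, creaks and shifts in the Helios desert wasteland biome.
# It mixes two random pools, thirteen and six waves, the smaller at half volume.
# Each pool gets pitch and volume modulation, nested looping envelopes and a quality switch with only the first two slots filled.
# It uses positional one-shot attenuation and allows one instance, preventing new plays while active.
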SoundCue /Game/Sounds/Helios/Biomes/DesertWasteland/Spots/Spot_Nitro_Wasteland_Debris_Creaks_Indoors_Int_Cue
root=Mixer_0
nodes [
 {
  id=Enveloper_0
  children [Enveloper_3]
  properties {
   bLoop=true
   bLoopIndefinitely=true
   loopEnd=100
   pitchCurve {editorCurveData {keys [<at 0 0.8>]}}
   volumeCurve {editorCurveData {keys [<at 0 1 RCIM_Cubic>,<at 51.3938 0.9352 RCIM_Cubic>,<at 100 1 RCIM_Cubic>]}}
  }
 }
 {
  id=Enveloper_1
  children [Enveloper_2]
  properties {
   bLoop=true
   bLoopIndefinitely=true
   loopEnd=100
   pitchCurve {editorCurveData {keys [<at 0 0.8>]}}
   volumeCurve {editorCurveData {keys [<at 0 1 RCIM_Cubic>,<at 51.3938 0.9352 RCIM_Cubic>,<at 100 1 RCIM_Cubic>]}}
  }
 }
 {
  id=Enveloper_2
  children [Modulator_1]
  properties {
   bLoop=true
   bLoopIndefinitely=true
   loopEnd=30
   pitchCurve {editorCurveData {keys [{interpMode=RCIM_Cubic, time=0}, <at 15.9468 0.9181 RCIM_Cubic>, {interpMode=RCIM_Cubic, time=30}]}}
   volumeCurve {editorCurveData {keys [<at 0 1>]}}
  }
 }
 {
  id=Enveloper_3
  children [Modulator_0]
  properties {
   bLoop=true
   bLoopIndefinitely=true
   loopEnd=30
   pitchCurve {editorCurveData {keys [{interpMode=RCIM_Cubic, time=0}, <at 15.9468 0.9181 RCIM_Cubic>, {interpMode=RCIM_Cubic, time=30}]}}
   volumeCurve {editorCurveData {keys [<at 0 1>]}}
  }
 }
 {id=Mixer_0, children [QualityLevel_0, QualityLevel_1], properties {inputVolume [1,0.5]}}
 {id=Modulator_0, children [Random_1], properties {pitchMax=1.2, pitchMin=0.7, volumeMax=1.1, volumeMin=0.9}}
 {id=Modulator_1, children [Random_4], properties {pitchMax=1.2, pitchMin=0.7, volumeMax=1.1, volumeMin=0.9}}
 {id=QualityLevel_0, children [Enveloper_0,Enveloper_0,null,null]}
 {id=QualityLevel_1, children [Enveloper_1,Enveloper_1,null,null]}
 {
  id=Random_1
  children [
   WavePlayer_23
   WavePlayer_24
   WavePlayer_25
   WavePlayer_26
   WavePlayer_27
   WavePlayer_28
   WavePlayer_29
   WavePlayer_30
   WavePlayer_31
   WavePlayer_42
   WavePlayer_43
   WavePlayer_44
   WavePlayer_45
  ]
  properties {weights [1,1,1,1,1,1,1,1,1,1,1,1,1]}
 }
 {
  id=Random_4
  children [WavePlayer_59, WavePlayer_60, WavePlayer_61, WavePlayer_62, WavePlayer_63, WavePlayer_64]
  properties {weights [1,1,1,1,1,1]}
 }
 {id=WavePlayer_23, properties {soundWaveAssetPtr=&/Spot_Interior_Debris_Fall_Creaks_Shift_13}}
 {id=WavePlayer_24, properties {soundWaveAssetPtr=&/Spot_Interior_Debris_Fall_Creaks_Shift_12}}
 {id=WavePlayer_25, properties {soundWaveAssetPtr=&/Spot_Interior_Debris_Fall_Creaks_Shift_11}}
 {id=WavePlayer_26, properties {soundWaveAssetPtr=&/Spot_Interior_Debris_Fall_Creaks_Shift_10}}
 {id=WavePlayer_27, properties {soundWaveAssetPtr=&/Spot_Interior_Debris_Fall_Creaks_Shift_09}}
 {id=WavePlayer_28, properties {soundWaveAssetPtr=&/Spot_Interior_Debris_Fall_Creaks_Shift_08}}
 {id=WavePlayer_29, properties {soundWaveAssetPtr=&/Spot_Interior_Debris_Fall_Creaks_Shift_07}}
 {id=WavePlayer_30, properties {soundWaveAssetPtr=&/Spot_Interior_Debris_Fall_Creaks_Shift_06}}
 {id=WavePlayer_31, properties {soundWaveAssetPtr=&/Spot_Interior_Debris_Fall_Creaks_Shift_05}}
 {id=WavePlayer_42, properties {soundWaveAssetPtr=&/Spot_Interior_Debris_Fall_Creaks_Shift_04}}
 {id=WavePlayer_43, properties {soundWaveAssetPtr=&/Spot_Interior_Debris_Fall_Creaks_Shift_03}}
 {id=WavePlayer_44, properties {soundWaveAssetPtr=&/Spot_Interior_Debris_Fall_Creaks_Shift_02}}
 {id=WavePlayer_45, properties {soundWaveAssetPtr=&/Spot_Interior_Debris_Fall_Creaks_Shift_01}}
 {id=WavePlayer_59, properties {soundWaveAssetPtr=&Spot_Interior_Debris_Fall_Creaks_Shift_01}}
 {id=WavePlayer_60, properties {soundWaveAssetPtr=&Spot_Interior_Debris_Fall_Creaks_Shift_02}}
 {id=WavePlayer_61, properties {soundWaveAssetPtr=&Spot_Interior_Debris_Fall_Creaks_Shift_03}}
 {id=WavePlayer_62, properties {soundWaveAssetPtr=&Spot_Interior_Debris_Fall_Creaks_Shift_04}}
 {id=WavePlayer_63, properties {soundWaveAssetPtr=&Spot_Interior_Debris_Fall_Creaks_Shift_05}}
 {id=WavePlayer_64, properties {soundWaveAssetPtr=&Spot_Interior_Debris_Fall_Creaks_Shift_06}}
]
properties {
 attenuationSettings=/Game/Sounds/Attenuation/Fort_TileAmbient_OneShot_MorePositional_Att
 bOverrideConcurrency=true
 maxDistance=8000
 soundClassObject=/Game/Sounds/FortSoundClassesAndModes/Fort_Ambient_OneShots_Indoors
 volumeMultiplier=0.251
 concurrencyOverrides {maxCount=1, resolutionRule=PreventNew}
 preEffectBusSends [{}]
}
