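# Indoor spot-ambience Sound Cue of creaking, stressed wood for the desert wasteland biome, using positional one-shot attenuation.
# A Mixer joins two quality-switched branches: the first two quality slots draw from 23 creak waves, the last two from 10 at half mixer volume.
# Each branch runs Random into pitch/volume Modulator, then two nested looping Envelopers shaping pitch and volume.
# A concurrency override caps playback at two instances, preventing new ones.
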SoundCue /Game/Sounds/Helios/Biomes/DesertWasteland/Spots/Spot_Nitro_Wasteland_Wooden_Stress_Creak_Int_Cue
root=Mixer_0
nodes [
 {
  id=Enveloper_0
  children [Enveloper_3]
  properties {
   bLoop=true
   bLoopIndefinitely=true
   loopEnd=100
   pitchCurve {editorCurveData {keys [<at 0 0.8>]}}
   volumeCurve {editorCurveData {keys [<at 0 1 RCIM_Cubic>,<at 51.3938 0.9352 RCIM_Cubic>,<at 100 1 RCIM_Cubic>]}}
  }
 }
 {
  id=Enveloper_1
  children [Enveloper_2]
  properties {
   bLoop=true
   bLoopIndefinitely=true
   loopEnd=100
   pitchCurve {editorCurveData {keys [<at 0 0.8>]}}
   volumeCurve {editorCurveData {keys [<at 0 1 RCIM_Cubic>,<at 51.3938 0.9352 RCIM_Cubic>,<at 100 1 RCIM_Cubic>]}}
  }
 }
 {
  id=Enveloper_2
  children [Modulator_1]
  properties {
   bLoop=true
   bLoopIndefinitely=true
   loopEnd=30
   pitchCurve {editorCurveData {keys [{interpMode=RCIM_Cubic, time=0}, <at 15.9468 0.9181 RCIM_Cubic>, {interpMode=RCIM_Cubic, time=30}]}}
   volumeCurve {editorCurveData {keys [<at 0 1>]}}
  }
 }
 {
  id=Enveloper_3
  children [Modulator_0]
  properties {
   bLoop=true
   bLoopIndefinitely=true
   loopEnd=30
   pitchCurve {editorCurveData {keys [{interpMode=RCIM_Cubic, time=0}, <at 15.9468 0.9181 RCIM_Cubic>, {interpMode=RCIM_Cubic, time=30}]}}
   volumeCurve {editorCurveData {keys [<at 0 1>]}}
  }
 }
 {id=Mixer_0, children [QualityLevel_0, QualityLevel_1], properties {inputVolume [1,0.5]}}
 {id=Modulator_0, children [Random_2], properties {pitchMax=1, pitchMin=0.5, volumeMax=1.1, volumeMin=0.9}}
 {id=Modulator_1, children [Random_0], properties {pitchMax=1, pitchMin=0.5, volumeMax=1.1, volumeMin=0.9}}
 {id=QualityLevel_0, children [Enveloper_0,Enveloper_0,null,null]}
 {id=QualityLevel_1, children [null,null,Enveloper_1,Enveloper_1]}
 {
  id=Random_0
  children [
   WavePlayer_0
   WavePlayer_1
   WavePlayer_2
   WavePlayer_3
   WavePlayer_4
   WavePlayer_18
   WavePlayer_19
   WavePlayer_20
   WavePlayer_21
   WavePlayer_22
  ]
  properties {weights [1,1,1,1,1,1,1,1,1,1]}
 }
 {
  id=Random_2
  children [
   WavePlayer_32
   WavePlayer_33
   WavePlayer_34
   WavePlayer_35
   WavePlayer_36
   WavePlayer_37
   WavePlayer_38
   WavePlayer_39
   WavePlayer_40
   WavePlayer_41
   WavePlayer_5
   WavePlayer_6
   WavePlayer_7
   WavePlayer_8
   WavePlayer_9
   WavePlayer_10
   WavePlayer_11
   WavePlayer_12
   WavePlayer_13
   WavePlayer_14
   WavePlayer_15
   WavePlayer_16
   WavePlayer_17
  ]
  properties {weights [1,1,1,1,1,1,1,1,1,1,1,1,1,1,1,1,1,1,1,1,1,1,1]}
 }
 {id=WavePlayer_0, properties {soundWaveAssetPtr=/Game/Sounds/Fort_Ambient_NonLoop/Primitive/Primitive_Structure_WoodCreaks_01}}
 {id=WavePlayer_1, properties {soundWaveAssetPtr=/Game/Sounds/Fort_Ambient_NonLoop/Primitive/Primitive_Structure_WoodCreaks_04}}
 {
  id=WavePlayer_10
  properties {soundWaveAssetPtr=&/Spot_Nitro_Wasteland_Wood_Creaks_Interior_06}
 }
 {
  id=WavePlayer_11
  properties {soundWaveAssetPtr=&/Spot_Nitro_Wasteland_Wood_Creaks_Interior_07}
 }
 {
  id=WavePlayer_12
  properties {soundWaveAssetPtr=&/Spot_Nitro_Wasteland_Wood_Creaks_Interior_08}
 }
 {
  id=WavePlayer_13
  properties {soundWaveAssetPtr=&/Spot_Nitro_Wasteland_Wood_Creaks_Interior_09}
 }
 {
  id=WavePlayer_14
  properties {soundWaveAssetPtr=&/Spot_Nitro_Wasteland_Wood_Creaks_Interior_10}
 }
 {
  id=WavePlayer_15
  properties {soundWaveAssetPtr=&/Spot_Nitro_Wasteland_Wood_Creaks_Interior_11}
 }
 {
  id=WavePlayer_16
  properties {soundWaveAssetPtr=&/Spot_Nitro_Wasteland_Wood_Creaks_Interior_12}
 }
 {
  id=WavePlayer_17
  properties {soundWaveAssetPtr=&/Spot_Nitro_Wasteland_Wood_Creaks_Interior_13}
 }
 {
  id=WavePlayer_18
  properties {soundWaveAssetPtr=&/Spot_Nitro_Wasteland_Wood_Creaks_Interior_01}
 }
 {
  id=WavePlayer_19
  properties {soundWaveAssetPtr=&/Spot_Nitro_Wasteland_Wood_Creaks_Interior_02}
 }
 {id=WavePlayer_2, properties {soundWaveAssetPtr=/Game/Sounds/Fort_Ambient_NonLoop/Primitive/Primitive_Structure_WoodCreaks_05}}
 {
  id=WavePlayer_20
  properties {soundWaveAssetPtr=&/Spot_Nitro_Wasteland_Wood_Creaks_Interior_03}
 }
 {
  id=WavePlayer_21
  properties {soundWaveAssetPtr=&/Spot_Nitro_Wasteland_Wood_Creaks_Interior_04}
 }
 {
  id=WavePlayer_22
  properties {soundWaveAssetPtr=&/Spot_Nitro_Wasteland_Wood_Creaks_Interior_05}
 }
 {id=WavePlayer_3, properties {soundWaveAssetPtr=/Game/Sounds/Fort_Ambient_NonLoop/Primitive/Primitive_Structure_WoodCreaks_10}}
 {id=WavePlayer_32, properties {soundWaveAssetPtr=&Primitive_Structure_WoodCreaks_01}}
 {id=WavePlayer_33, properties {soundWaveAssetPtr=&Primitive_Structure_WoodCreaks_04}}
 {id=WavePlayer_34, properties {soundWaveAssetPtr=&Primitive_Structure_WoodCreaks_05}}
 {id=WavePlayer_35, properties {soundWaveAssetPtr=&Primitive_Structure_WoodCreaks_10}}
 {id=WavePlayer_36, properties {soundWaveAssetPtr=/Game/Sounds/Fort_Ambient_NonLoop/Primitive/Primitive_Structure_WoodCreaks_02}}
 {id=WavePlayer_37, properties {soundWaveAssetPtr=/Game/Sounds/Fort_Ambient_NonLoop/Primitive/Primitive_Structure_WoodCreaks_03}}
 {id=WavePlayer_38, properties {soundWaveAssetPtr=/Game/Sounds/Fort_Ambient_NonLoop/Primitive/Primitive_Structure_WoodCreaks_06}}
 {id=WavePlayer_39, properties {soundWaveAssetPtr=/Game/Sounds/Fort_Ambient_NonLoop/Primitive/Primitive_Structure_WoodCreaks_07}}
 {id=WavePlayer_4, properties {soundWaveAssetPtr=&Primitive_Structure_WoodCreaks_02}}
 {id=WavePlayer_40, properties {soundWaveAssetPtr=/Game/Sounds/Fort_Ambient_NonLoop/Primitive/Primitive_Structure_WoodCreaks_08}}
 {id=WavePlayer_41, properties {soundWaveAssetPtr=/Game/Sounds/Fort_Ambient_NonLoop/Primitive/Primitive_Structure_WoodCreaks_09}}
 {
  id=WavePlayer_5
  properties {soundWaveAssetPtr=&Spot_Nitro_Wasteland_Wood_Creaks_Interior_01}
 }
 {
  id=WavePlayer_6
  properties {soundWaveAssetPtr=&Spot_Nitro_Wasteland_Wood_Creaks_Interior_02}
 }
 {
  id=WavePlayer_7
  properties {soundWaveAssetPtr=&Spot_Nitro_Wasteland_Wood_Creaks_Interior_03}
 }
 {
  id=WavePlayer_8
  properties {soundWaveAssetPtr=&Spot_Nitro_Wasteland_Wood_Creaks_Interior_04}
 }
 {
  id=WavePlayer_9
  properties {soundWaveAssetPtr=&Spot_Nitro_Wasteland_Wood_Creaks_Interior_05}
 }
]
properties {
 attenuationSettings=/Game/Sounds/Attenuation/Fort_TileAmbient_OneShot_MorePositional_Att
 bOverrideConcurrency=true
 maxDistance=8000
 soundClassObject=/Game/Sounds/FortSoundClassesAndModes/Fort_Ambient_OneShots_Indoors
 volumeMultiplier=0.501
 concurrencyOverrides {maxCount=2, resolutionRule=PreventNew}
 preEffectBusSends [{}]
}
