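# Ambient spot Sound Cue for the DesertWasteland biome that plays looping wind-driven metal box rattles.
# All quality levels share one chain: a random start Delay feeds a Looping node over an equally weighted Random pick of six gusty shipping-container recordings.
# The Delay sits outside the loop, so it applies only before the first rattle.
# A box-shaped attenuation override bounds the sound, which uses an indoor ambient sound class.
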SoundCue /Game/Sounds/Helios/Biomes/DesertWasteland/Spots/Spot_Wasteland_Metal_Box_Wind_Rattles_Focus_Scaling_Distant_Looping_Cue
root=QualityLevel_0
nodes [
 {id=Delay_0, children [Looping_0], properties {delayMax=12, delayMin=4}}
 {id=Looping_0, children [Random_0]}
 {id=QualityLevel_0, children [Delay_0,Delay_0,Delay_0,Delay_0]}
 {id=Random_0, children [WavePlayer_0,WavePlayer_1,WavePlayer_4,WavePlayer_5,WavePlayer_6,WavePlayer_7], properties {weights [1,1,1,1,1,1]}}
 {
  id=WavePlayer_0
  properties {soundWaveAssetPtr=/Game/Sounds/Asteria/Environment_Actors/MetalShippingContainer/Spot_Wind_Metal_Ratte_Gusty_Interior_01}
 }
 {
  id=WavePlayer_1
  properties {soundWaveAssetPtr=/Game/Sounds/Asteria/Environment_Actors/MetalShippingContainer/Spot_Wind_Metal_Ratte_Gusty_Interior_02}
 }
 {
  id=WavePlayer_4
  properties {soundWaveAssetPtr=/Game/Sounds/Asteria/Environment_Actors/MetalShippingContainer/Spot_Wind_Metal_Ratte_Gusty_Interior_03}
 }
 {
  id=WavePlayer_5
  properties {soundWaveAssetPtr=/Game/Sounds/Asteria/Environment_Actors/MetalShippingContainer/Spot_Wind_Metal_Ratte_Gusty_Interior_04}
 }
 {
  id=WavePlayer_6
  properties {soundWaveAssetPtr=/Game/Sounds/Asteria/Environment_Actors/MetalShippingContainer/Spot_Wind_Metal_Ratte_Gusty_Interior_05}
 }
 {
  id=WavePlayer_7
  properties {soundWaveAssetPtr=/Game/Sounds/Asteria/Environment_Actors/MetalShippingContainer/Spot_Wind_Metal_Ratte_Gusty_Interior_06}
 }
]
properties {
 bOverrideAttenuation=true
 maxDistance=900.4811
 soundClassObject=&Fort_Ambient_Indoors_Athena
 volumeMultiplier=0.501
 attenuationOverrides {attenuationShape=Box, falloffDistance=250, attenuationShapeExtents=<500.1809 507.2498 650.4811>}
 preEffectBusSends [{}]
}
